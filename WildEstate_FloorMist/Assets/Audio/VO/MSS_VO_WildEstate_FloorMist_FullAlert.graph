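# MetaSound preset of the shared boss VO source that plays full-alert voice lines for FloorMist NPCs in WildEstate, using HoundsHenchmen recordings.
# It supplies seven waves to the parent graph as a mono one-shot, with delay, pitch and volume randomization zeroed and probability at full.
# Output uses alerted-NPC attenuation and the non-critical enemy VO class, and feeds music, weapon and explosion sidechain buses.
# Boss combat-chatter concurrency is set, alongside a one-per-owner, prevent-new override.
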MetaSoundSource /WildEstate_FloorMist/Assets/Audio/VO/MSS_VO_WildEstate_FloorMist_FullAlert
preset_parent=/Game/Sounds/Fort_Human_VO_BR/NPCs/Shared/MSS_VO_BR_Boss
inputs [
 "Delay Max (Seconds)" {type=Time}
 "Delay Min (Seconds)" {type=Time}
 "Probability (0-1)" {type=Float}
 "Random Pitch Max (Semitones)" {type=Float}
 "Random Pitch Min (Semitones)" {type=Float}
 "Random Volume Change Max (Decibels)" {type=Float}
 "Random Volume Change Min (Decibels)" {type=Float}
 UE.Source.OnPlay {type=Trigger}
 Wave Array {type=WaveAsset:Array}
]
interfaces [UE.OutputFormat.Mono:1.0, UE.Source.OneShot:1.0, UE.Source:1.1]
outputs [
 "UE.OutputFormat.Mono.Audio:0" {type=Audio}
 UE.Source.OneShot.OnFinished {type=Trigger}
]
pages [Default {input_defaults {
   "Delay Max (Seconds)"=0
   "Delay Min (Seconds)"=0
   "Probability (0-1)"=1
   "Random Pitch Max (Semitones)"=0
   "Random Pitch Min (Semitones)"=0
   "Random Volume Change Max (Decibels)"=0
   "Random Volume Change Min (Decibels)"=0
   UE.Source.OnPlay=null
   Wave Array [
    /WildEstate_FloorMist/Assets/Audio/VO/Waves/VO_S38_HoundsHenchmen_FullAlert_010
    /WildEstate_FloorMist/Assets/Audio/VO/Waves/VO_S38_HoundsHenchmen_FullAlert_011
    /WildEstate_FloorMist/Assets/Audio/VO/Waves/VO_S38_HoundsHenchmen_FullAlert_020
    /WildEstate_FloorMist/Assets/Audio/VO/Waves/VO_S38_HoundsHenchmen_FullAlert_021
    /WildEstate_FloorMist/Assets/Audio/VO/Waves/VO_S38_HoundsHenchmen_FullAlert_030
    /WildEstate_FloorMist/Assets/Audio/VO/Waves/VO_S38_HoundsHenchmen_FullAlert_040
    /WildEstate_FloorMist/Assets/Audio/VO/Waves/VO_S38_HoundsHenchmen_FullAlert_050
   ]
  }}]
properties {
 assetImportData=&self:AssetImportData
 attenuationSettings=/Game/Sounds/Fort_Human_VO_BR/NPCs/Shared/NPC_VO_Alerted_Att
 maxDistance=12500
 sampleRate=48000
 soundClassObject=/Game/Sounds/FortSoundClassesAndModes/Fort_VO_Enemy_NonCritical
 volume=2.6738
 busSends [
  {audioBus=&AB_Sidechain_FortMusic_Vol, sendLevel=0.25}
  {audioBus=/Game/Sounds/Modulation/EnvelopeFollowers/Volume/AB_Sidechain_FortWeapon1P_Vol, sendLevel=0.25}
  {audioBus=&AB_Sidechain_FortWeapon_Vol, sendLevel=0.25}
  {audioBus=&AB_Sidechain_FortExplosion_Vol, sendLevel=0.25}
 ]
 concurrencyOverrides {bLimitToOwner=true, maxCount=1, resolutionRule=PreventNew}
 concurrencySet [&NPC_Boss_Combat_Chatter_Conc]
}
